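# Animation Blueprint for the ThickWatch backpack cosmetic, playing an elimination reaction in gameplay and a reactive animation in the frontend.
# A bool blend picks between the gameplay and frontend state machines, each returning to idle afterwards.
# An enemy-elimination anim input event briefly pulses the gameplay trigger, while the backpack part modifier enables a delayed frontend trigger that re-arms via state-entry notify.
# An unconnected combat state machine still references OysterKnock animations.
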AnimBlueprint /BRCosmetics/Accessories/FORT_Backpacks/Backpack_M_MED_ThickWatch/Meshes/M_MED_ThickWatch_Pack_AnimBP
target_skeleton=&/M_MED_ThickWatch_Pack_Skeleton
graphs [
 AnimGraph {nodes [
   Input Pose()
   AnimGraphNode_StateMachine_0 = ThickWatch Backpack States() {is=title AnimGraphNode_StateMachine}
   AnimGraphNode_StateMachine_1 = FE_Reactivity() {is=title AnimGraphNode_StateMachine}
   n1 = Blend Poses by bool(
    BlendPose_0=@AnimGraphNode_StateMachine_1
    BlendPose_1=@AnimGraphNode_StateMachine_0
    bActiveValue=Get bShouldDisplayFrontendReactivity() -> (bShouldDisplayFrontendReactivity:bool)
   ) {blendType=Linear}
   Output Pose(Result=Component To Local(ComponentPose=Local To Component(LocalPose=@n1))) {layerGroup=None}
   AnimGraphNode_StateMachine_2 = Backpack States() {is=title AnimGraphNode_StateMachine}
  ]}
 ThickWatch Backpack States {
  entry=Idle
  path=AnimGraph.AnimGraphNode_StateMachine_0.ThickWatch Backpack States
  ids=id
  nodes [
   AnimStateNode_0 = Idle()
   AnimStateNode_1 = Elimination() {is=title AnimStateNode}
   AnimStateTransitionNode_0 = AnimStateTransitionNode()
   AnimStateTransitionNode_1 = AnimStateTransitionNode() {bAutomaticRuleBasedOnSequencePlayerInState=true}
  ]
  links [
   {
    dst=Idle
    rule=AnimGraph.AnimGraphNode_StateMachine_0.ThickWatch Backpack States.AnimStateTransitionNode_1.Transition
    src=Elimination
    via=AnimStateTransitionNode_1
   }
   {
    dst=Elimination
    rule=AnimGraph.AnimGraphNode_StateMachine_0.ThickWatch Backpack States.AnimStateTransitionNode_0.Transition
    src=Idle
    via=AnimStateTransitionNode_0
   }
  ]
 }
 Idle {
  path=AnimGraph.AnimGraphNode_StateMachine_0.ThickWatch Backpack States.AnimStateNode_0.Idle
  nodes [Output Animation Pose(Result=Sequence Player 'ThickWatch_IdlePose_PACK'() {bLoopAnimation=false, sequence=/BRCosmetics/Accessories/FORT_Backpacks/Backpack_M_MED_ThickWatch/Animation/ThickWatch_IdlePose_PACK})]
 }
 Elimination {
  path=AnimGraph.AnimGraphNode_StateMachine_0.ThickWatch Backpack States.AnimStateNode_1.Elimination
  nodes [
   n1 = Sequence Player 'ThickWatch_ElimReact_PACK'() {bLoopAnimation=false, sequence=/BRCosmetics/Accessories/FORT_Backpacks/Backpack_M_MED_ThickWatch/Animation/ThickWatch_ElimReact_PACK}
   Output Animation Pose(Result=@n1)
  ]
 }
 Transition {
  path=AnimGraph.AnimGraphNode_StateMachine_0.ThickWatch Backpack States.AnimStateTransitionNode_0.Transition
  nodes [Result(bCanEnterTransition=Get bEnemyPlayerEliminated() -> (bEnemyPlayerEliminated:bool))]
 }
 Transition {path=AnimGraph.AnimGraphNode_StateMachine_0.ThickWatch Backpack States.AnimStateTransitionNode_1.Transition, nodes [Result()]}
 FE_Reactivity {
  entry=Idle
  path=AnimGraph.AnimGraphNode_StateMachine_1.FE_Reactivity
  ids=id
  nodes [
   AnimStateNode_0 = Idle()
   AnimStateNode_2 = FE_Elim() {is=title AnimStateNode, stateEntered {guid=#1, notifyName=Entered FE_Reactive_Anim}}
   AnimStateTransitionNode_8 = AnimStateTransitionNode()
   AnimStateTransitionNode_9 = AnimStateTransitionNode() {bAutomaticRuleBasedOnSequencePlayerInState=true}
  ]
  links [
   {
    dst=Idle
    rule=AnimGraph.AnimGraphNode_StateMachine_1.FE_Reactivity.AnimStateTransitionNode_9.Transition
    src=FE_Elim
    via=AnimStateTransitionNode_9
   }
   {
    dst=FE_Elim
    rule=AnimGraph.AnimGraphNode_StateMachine_1.FE_Reactivity.AnimStateTransitionNode_8.Transition
    src=Idle
    via=AnimStateTransitionNode_8
   }
  ]
 }
 Idle {
  path=AnimGraph.AnimGraphNode_StateMachine_1.FE_Reactivity.AnimStateNode_0.Idle
  nodes [Output Animation Pose(Result="Sequence Evaluator (by time) 'ThickWatch_IdlePose_PACK'"!(ExplicitTime:float=0) -> (Pose:Pose Link Structure) {
     is=title AnimGraphNode_SequenceEvaluator
     sequence=&ThickWatch_IdlePose_PACK
    })]
 }
 FE_Elim {
  path=AnimGraph.AnimGraphNode_StateMachine_1.FE_Reactivity.AnimStateNode_2.FE_Elim
  nodes [
   n1 = Sequence Player 'ThickWatch_FE_Reactive_PACK'() {bLoopAnimation=false, sequence=/BRCosmetics/Accessories/FORT_Backpacks/Backpack_M_MED_ThickWatch/Animation/ThickWatch_FE_Reactive_PACK}
   Output Animation Pose(Result=@n1)
  ]
 }
 Transition {
  path=AnimGraph.AnimGraphNode_StateMachine_1.FE_Reactivity.AnimStateTransitionNode_8.Transition
  nodes [Result(bCanEnterTransition=Get Play Reactive FE Anim() -> (Play Reactive FE Anim:bool))]
 }
 Transition {path=AnimGraph.AnimGraphNode_StateMachine_1.FE_Reactivity.AnimStateTransitionNode_9.Transition, nodes [Result()]}
 Backpack States {
  entry=Idle_pose
  path=AnimGraph.AnimGraphNode_StateMachine_2.Backpack States
  ids=id
  nodes [
   AnimStateNode_0 = Idle_pose() {
    is=title AnimStateNode
    stateEntered {guid=#2, notifyName=Entered_RandomLocoBreak}
    stateFullyBlended {guid=#3}
    stateLeft {guid=#4}
   }
   AnimStateNode_12 = Into_Combat_Pose() {
    is=title AnimStateNode
    stateEntered {guid=#5}
    stateFullyBlended {guid=#6}
    stateLeft {guid=#7}
   }
   AnimStateNode_15 = Elimination() {
    stateEntered {guid=#8}
    stateFullyBlended {guid=#9}
    stateLeft {guid=#10}
   }
   AnimStateNode_2 = Exit Combat Pose() {
    is=title AnimStateNode
    stateEntered {guid=#11}
    stateFullyBlended {guid=#12}
    stateLeft {guid=#13}
   }
   AnimStateNode_5 = Combat_pose() {
    is=title AnimStateNode
    stateEntered {guid=#14}
    stateFullyBlended {guid=#15}
    stateLeft {guid=#16}
   }
   AnimStateTransitionNode_1 = AnimStateTransitionNode()
   AnimStateTransitionNode_10 = AnimStateTransitionNode()
   AnimStateTransitionNode_16 = AnimStateTransitionNode()
   AnimStateTransitionNode_21 = AnimStateTransitionNode() {bAutomaticRuleBasedOnSequencePlayerInState=true}
   AnimStateTransitionNode_24 = AnimStateTransitionNode()
   AnimStateTransitionNode_27 = AnimStateTransitionNode() {bAutomaticRuleBasedOnSequencePlayerInState=true}
   AnimStateTransitionNode_34 = AnimStateTransitionNode() {
    transitionEnd {guid=#17}
    transitionInterrupt {guid=#18}
    transitionStart {guid=#19}
   }
   AnimStateTransitionNode_36 = AnimStateTransitionNode()
   AnimStateTransitionNode_4 = AnimStateTransitionNode() {
    bAutomaticRuleBasedOnSequencePlayerInState=true
    transitionEnd {guid=#20}
    transitionInterrupt {guid=#21}
    transitionStart {guid=#22}
   }
  ]
  links [
   {
    dst=Elimination
    rule=AnimGraph.AnimGraphNode_StateMachine_2.Backpack States.AnimStateTransitionNode_34.Transition
    src=Combat_pose
    via=AnimStateTransitionNode_34
   }
   {
    dst=Exit Combat Pose
    rule=AnimGraph.AnimGraphNode_StateMachine_2.Backpack States.AnimStateTransitionNode_10.Transition
    src=Combat_pose
    via=AnimStateTransitionNode_10
   }
   {
    dst=Idle_pose
    rule=AnimGraph.AnimGraphNode_StateMachine_2.Backpack States.AnimStateTransitionNode_27.Transition
    src=Elimination
    via=AnimStateTransitionNode_27
   }
   {
    dst=Combat_pose
    rule=AnimGraph.AnimGraphNode_StateMachine_2.Backpack States.AnimStateTransitionNode_1.Transition
    src=Exit Combat Pose
    via=AnimStateTransitionNode_1
   }
   {
    dst=Elimination
    rule=AnimGraph.AnimGraphNode_StateMachine_2.Backpack States.AnimStateTransitionNode_36.Transition
    src=Exit Combat Pose
    via=AnimStateTransitionNode_36
   }
   {
    dst=Idle_pose
    rule=AnimGraph.AnimGraphNode_StateMachine_2.Backpack States.AnimStateTransitionNode_21.Transition
    src=Exit Combat Pose
    via=AnimStateTransitionNode_21
   }
   {
    dst=Into_Combat_Pose
    rule=AnimGraph.AnimGraphNode_StateMachine_2.Backpack States.AnimStateTransitionNode_16.Transition
    src=Idle_pose
    via=AnimStateTransitionNode_16
   }
   {
    dst=Combat_pose
    rule=AnimGraph.AnimGraphNode_StateMachine_2.Backpack States.AnimStateTransitionNode_4.Transition
    src=Into_Combat_Pose
    via=AnimStateTransitionNode_4
   }
   {
    dst=Elimination
    rule=AnimGraph.AnimGraphNode_StateMachine_2.Backpack States.AnimStateTransitionNode_24.Transition
    src=Into_Combat_Pose
    via=AnimStateTransitionNode_24
   }
  ]
 }
 Idle_pose {
  path=AnimGraph.AnimGraphNode_StateMachine_2.Backpack States.AnimStateNode_0.Idle_pose
  nodes [Output Animation Pose(Result=Sequence Player 'OysterKnock_IdlePose_pack'() {bLoopAnimation=false, sequence=/BRCosmetics/Accessories/FORT_Backpacks/Backpack_M_MED_OysterKnock/Anim/OysterKnock_IdlePose_pack})]
 }
 Into_Combat_Pose {
  path=AnimGraph.AnimGraphNode_StateMachine_2.Backpack States.AnimStateNode_12.Into_Combat_Pose
  nodes [Output Animation Pose(Result=Sequence Player 'OysterKnock_Combat_into_pack'() {bLoopAnimation=false, sequence=/BRCosmetics/Accessories/FORT_Backpacks/Backpack_M_MED_OysterKnock/Anim/OysterKnock_Combat_into_pack})]
 }
 Elimination {
  path=AnimGraph.AnimGraphNode_StateMachine_2.Backpack States.AnimStateNode_15.Elimination
  nodes [Output Animation Pose(Result=Sequence Player 'OysterKnock_Elim_pack'() {bLoopAnimation=false, sequence=/BRCosmetics/Accessories/FORT_Backpacks/Backpack_M_MED_OysterKnock/Anim/OysterKnock_Elim_pack})]
 }
 Exit Combat Pose {
  path=AnimGraph.AnimGraphNode_StateMachine_2.Backpack States.AnimStateNode_2.Exit Combat Pose
  nodes [Output Animation Pose(Result=Sequence Player 'OysterKnock_Combat_out_pack'() {bLoopAnimation=false, sequence=/BRCosmetics/Accessories/FORT_Backpacks/Backpack_M_MED_OysterKnock/Anim/OysterKnock_Combat_out_pack})]
 }
 Combat_pose {
  path=AnimGraph.AnimGraphNode_StateMachine_2.Backpack States.AnimStateNode_5.Combat_pose
  nodes [Output Animation Pose(Result=Sequence Player 'OysterKnock_CombatPose_pack'() {bLoopAnimation=false, sequence=/BRCosmetics/Accessories/FORT_Backpacks/Backpack_M_MED_OysterKnock/Anim/OysterKnock_CombatPose_pack})]
 }
 Transition {
  path=AnimGraph.AnimGraphNode_StateMachine_2.Backpack States.AnimStateTransitionNode_1.Transition
  nodes [Result(bCanEnterTransition=Get isCombat() -> (isCombat:bool))]
 }
 Transition {
  path=AnimGraph.AnimGraphNode_StateMachine_2.Backpack States.AnimStateTransitionNode_10.Transition
  nodes [Result(bCanEnterTransition=NOT Boolean(A=Get isCombat()))]
 }
 Transition {
  path=AnimGraph.AnimGraphNode_StateMachine_2.Backpack States.AnimStateTransitionNode_16.Transition
  nodes [Result(bCanEnterTransition=Get isCombat())]
 }
 Transition {
  path=AnimGraph.AnimGraphNode_StateMachine_2.Backpack States.AnimStateTransitionNode_21.Transition
  nodes [Result(bCanEnterTransition=NOT Boolean(A=Get isCombat()))]
 }
 Transition {
  path=AnimGraph.AnimGraphNode_StateMachine_2.Backpack States.AnimStateTransitionNode_24.Transition
  nodes [Result(bCanEnterTransition=Get bEnemyPlayerEliminated())]
 }
 Transition {path=AnimGraph.AnimGraphNode_StateMachine_2.Backpack States.AnimStateTransitionNode_27.Transition, nodes [Result()]}
 Transition {
  path=AnimGraph.AnimGraphNode_StateMachine_2.Backpack States.AnimStateTransitionNode_34.Transition
  nodes [Result(bCanEnterTransition=Get bEnemyPlayerEliminated())]
 }
 Transition {
  path=AnimGraph.AnimGraphNode_StateMachine_2.Backpack States.AnimStateTransitionNode_36.Transition
  nodes [Result(bCanEnterTransition=Get bEnemyPlayerEliminated())]
 }
 Transition {path=AnimGraph.AnimGraphNode_StateMachine_2.Backpack States.AnimStateTransitionNode_4.Transition, nodes [Result()]}
 EventGraph {nodes [
   Event OnCharacterCustomizationLoaded()
   > n1 = Cast To FortPlayerPawn(Object=TryGetFortPawn())
   > n2 = Cast To B_Athena_PartModifier_Backpack_ThickWatch!(
    execute
    Object=GetModifierForPartType!(self:Fort Player Pawn Object Reference=@n1, PartType:EFortCustomPartType Enum=Backpack) -> (ReturnValue:Custom Character Part Modifier Object Reference)
   ) -> (then, CastFailed, AsB Athena Part Modifier Backpack Thick Watch:B Athena Part Modifier Backpack Thick Watch Object Reference) {is=title K2Node_DynamicCast}
   > n3 = Branch(Condition=Get FE_ReactivityEnabled!(self:B Athena Part Modifier Backpack Thick Watch Object Reference=@n2) -> (FE_ReactivityEnabled:bool))
   > Do Once(Reset=@n5.then)
   > n4 = Delay(Duration=4)
   Set Play Reactive FE Anim(execute=[@n7.then, @n4.then], Play Reactive FE Anim:bool=true) -> (Output_Get:bool)
   n5 = Set Play Reactive FE Anim(execute=@n3.else, Play Reactive FE Anim=false)
   n6 = Event BlueprintProcessAnimInputEvent()
   > Branch(Condition=IsNamed(self=@n6.AnimInputEvent, InEventName=aie.EnemyPlayerEliminated))
   > Set bEnemyPlayerEliminated(bEnemyPlayerEliminated:bool=true) -> (Output_Get:bool)
   > Delay(Duration=0.2)
   > Set bEnemyPlayerEliminated(bEnemyPlayerEliminated=false)
   Event BlueprintPostEvaluateAnimation() {is=title K2Node_Event}
   > Set bEnemyPlayerEliminated(bEnemyPlayerEliminated=false)
   AnimNotify_Entered FE_Reactive_Anim() {is=title K2Node_Event}
   > Set Play Reactive FE Anim(Play Reactive FE Anim=false)
   > n7 = Delay(Duration=10)
   TryGetPawnOwner()
   Event BlueprintUpdateAnimation()
  ]}
]
variables [
 bEnemyPlayerEliminated {type=boolean}
 Play Reactive FE Anim {type=boolean}
]
